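; Foraged Spooky Mist item actor that restricts which pawns can interact with it.
; BlueprintCanInteract allows only player pawns not holding an object, without the Spooky Mist effect or active Zero Point ability, and neither being nor controlling a remote-controlled pawn.
; On begin interact, a pawn tagged as holding an object switches off the replicated can-interact state; otherwise the parent class handles the interaction.
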(event Interaction|EventBlueprintOnBeginInteract
  (params (InteractingPawn :type Fort%20Pawn%20Object%20Reference) (InteractionBeingAttempted :type EInteractionBeingAttempted%20Enum :default "FirstInteraction") (InteractionType :type TInteractionType%20Enum :default "IT_NoInteraction"))
  (if (GameplayTags|HasMatchingGameplayTag InteractingPawn "(TagName=\"Granted.HeldObject.Held\")")
    (Variables|Interact|SetReplicatedCanInteract false)
    (else
      (|CallParentFunction InteractingPawn InteractionBeingAttempted InteractionType))))

(fn BlueprintCanInteract
  (params (InteractingPawn :type Fort%20Pawn%20Object%20Reference) (InteractionBeingAttempted :type EInteractionBeingAttempted%20Enum :default "FirstInteraction") (InteractionType :type TInteractionType%20Enum :default "IT_NoInteraction"))
  (bind (_asfort_remote_controlled_pawn_athena _bsuccess) (Utilities|Casting|CastToFortRemoteControlledPawnAthena (LocalVariables|GetPlayerPawn)))
  (bind _asfort_player_pawn (Utilities|Casting|CastToFortPlayerPawn InteractingPawn)
    (:then
      (bind _output_get (LocalVariables|SetPlayerPawn _asfort_player_pawn))
      (bind _returnvalue (Fortnite|HeldObjects|GetHeldObjectAttachedtoPlayer _output_get))
      (bind _returnvalue_2 (Fortnite|HeldObjects|IsHeldObjectBeingHeld _returnvalue))
      (if (not _returnvalue_2)
        (return (Math|Boolean|ANDBoolean (not (GameplayTags|HasMatchingGameplayTag (LocalVariables|GetPlayerPawn) "(TagName=\"GameplayEffect.ForagedItem.SpookyMist\")")) (not _bsuccess) (not (RemoteControlPawn|HasControlledRCPawn (LocalVariables|GetPlayerPawn))) (not (GameplayTags|HasMatchingGameplayTag (LocalVariables|GetPlayerPawn) "(TagName=\"Ability.Athena.ZeroPoint.AbilityActive\")"))))))
    (:CastFailed
      (return false))))

(fn ConstructionScript
  (|CallParentFunction))
