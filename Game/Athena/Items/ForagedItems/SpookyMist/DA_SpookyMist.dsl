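; Actor for the Spooky Mist foraged item that cancels the effect when its instigating player goes down but not out (DBNO).
; On BeginPlay it calls SetWeaponVisibility on the instigator's current weapon and checks the player's DBNO state.
; If already DBNO it immediately sends the generic triggered-ability Activate gameplay event to the instigator, which serves as the cancel signal.
; Otherwise it binds to OnPlayerStartDBNO and sends the same event when that fires.
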(event EventBeginPlay
  (bind _returnvalue (Game|GetInstigator))
  (bind _asfort_player_pawn_athena (Utilities|Casting|CastToFortPlayerPawnAthena _returnvalue))
  ;[1 1 1] If player is DBNO send the cancel event
  (FortWeapon|SetWeaponVisibility (Variables|/Script/FortniteGame.FortPawn|GetCurrentWeapon (Utilities|Casting|CastToFortPlayerPawn (Game|GetInstigator))) true true)
  (if (Variables|/Script/FortniteGame.FortPawn|GetbIsDBNO _asfort_player_pawn_athena)
    (Ability|SendGameplayEventtoActor _returnvalue "(TagName=\"Abilities.Generic.TriggeredAbility.Activate\")")
    (else
      (EventDispatchers|BindEventtoOnPlayerStartDBNO _asfort_player_pawn_athena (delegate Custom|On%20DBNO)))))

(event Custom|On%20DBNO
  (Ability|SendGameplayEventtoActor (Game|GetInstigator) "(TagName=\"Abilities.Generic.TriggeredAbility.Activate\")"))

(fn ConstructionScript)
