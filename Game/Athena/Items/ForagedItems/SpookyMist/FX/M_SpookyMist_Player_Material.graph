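# Translucent unlit skeletal-mesh material that gives the player a smoky, shifting Spooky Mist appearance.
# World-aligned noise pans mostly relative to the actor, shaping Fresnel opacity and dark-to-light color lit by sun/moon direction and time of day.
# Vertex colors mask hands, legs and feet, a timestamped eased flash adds color, and movement plus collection parameters drive opacity.
# A panning distortion normal adds detail, with cheaper Low-quality and Mobile branches.
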Material /Game/Athena/Items/ForagedItems/SpookyMist/FX/M_SpookyMist_Player_Material
graph {
 properties {
  bUsedWithMeshDeformer=true
  bUsedWithSkeletalMesh=true
  blendMode=BLEND_Translucent
  pixelDepthOffsetMode=PDOM_Legacy
  refractionMethod=RM_IndexOfRefraction
  shadingModel=MSM_Unlit
  usedShadingModels=Unlit
 }
 nodes [
  n1 = Time() {period=5}
  n2 = ScalarParameter() {parameterName=FlashStartTime}
  n3 = Clamp(None=Divide(A=Subtract(A=@n1, B=@n2), B=ScalarParameter() {defaultValue=0.5, group=Timestamp, parameterName=FlashDuration}))
  n4 = VectorParameter() {parameterName=Hand/LegColor, defaultValue=<rgba 0.0117 0.0199 0.05 1>}
  n5 = VertexColor()
  n6 = Add(A=@n5.R, B=Multiply(A=@n5.G) {constB=3.5})
  n7 = CollectionParameter() {
   collection=&FortniteMaterialParameters
   parameterName=SunAndMoonModelDirectionalVector
  }
  n8 = Clamp(None=DotProduct(A=PixelNormalWS(), B=ComponentMask(None=@n7) {b=true, g=true, r=true}))
  n9 = Multiply(A=MaterialFunctionCall() {materialFunction=&Color_Time_of_Day}.Result) {constB=0.07}
  n10 = VectorParameter() {parameterName=LightColorHighlight, defaultValue=<rgba 0.1313 0.0405 0.3 0>}
  n11 = Multiply(A=OneMinus(None=VertexColor().B), B=Multiply(A=@n8, B=Add(A=@n9, B=@n10.RGB)))
  n12 = Multiply(A=@n10.RGB) {constB=0.25}
  n13 = VectorParameter() {parameterName=DarkColor, defaultValue=<rgba 0.0575 0.0701 0.16 0>}
  n14 = VectorParameter() {parameterName=LightColor, defaultValue=<rgba 0.0673 0.0536 0.2 0>}
  n15 = Multiply(A=ScalarParameter() {defaultValue=20, parameterName=Scale}, B=VectorParameter() {parameterName=ScaleMult, defaultValue=<rgba 1 1 2.2 0>}.RGB)
  n16 = Multiply(A=Time(), B=ScalarParameter() {defaultValue=32, parameterName=PanSpeed})
  n17 = Multiply(A=@n16, B=Constant3Vector() {constant=<rgba 0.08 -0.08 1 0>})
  n18 = Subtract(A=WorldPosition().XYZ, B=Multiply(A=ActorPositionWS()) {constB=0.95})
  n19 = ComponentMask(None=MaterialFunctionCall(
    TextureObject=TextureObject() {texture=&T_Smooth_tile}
    TextureSize=@n15
    WorldPosition=Add(A=@n17, B=@n18)
   ) {materialFunction=&WorldAlignedTexture}.XYZ Texture) {g=true}
  n20 = LinearInterpolate(A=@n19, Alpha=QualitySwitch(Default=@n8, Low=Constant() {r=0.3})) {constB=0.84}
  n21 = VertexColor()
  n22 = Add(A=Add(A=Multiply(A=@n21.R) {constB=5}, B=@n21.B)) {constB=0.2}
  n23 = OneMinus(None=Fresnel(ExponentIn=LinearInterpolate(A=@n22, Alpha=@n20) {constA=0.2, constAlpha=1, constB=2.2}, Normal=VertexNormalWS()) {baseReflectFraction=0})
  n24 = ScalarParameter() {defaultValue=2, parameterName=FresnelPower}
  n25 = Add(A=@n19) {constB=0.3}
  n26 = Multiply(
   A=ShadingPathSwitch(Default=Power(Base=@n23, Exp=@n24) {constExponent=1.9432}, Mobile=Add(A=@n25, B=@n21.R)) {desc=Migrated from FeatureLevelSwitch}
   B=LinearInterpolate(B=ScalarParameter() {parameterName=Foot Opacity}, Alpha=@n21.G) {constA=1}
  )
  n27 = Saturate(None=Multiply(A=@n20, B=@n26))
  n28 = Multiply(
   A=Add(A=QualitySwitch(Default=@n11, Low=@n12), B=LinearInterpolate(A=@n13.RGB, B=@n14.RGB, Alpha=@n27))
   B=Add(A=@n27, B=ScalarParameter() {defaultValue=0.25, parameterName=ColorAdd}) {constB=0.05}
  )
  n29 = OneMinus(None=@n5.R)
  Add(
   A=Multiply(
    A=MaterialFunctionCall!(In=@n3) {materialFunction=/Game/Effects/Fort_Effects/Materials/_Functions/MF_Smoothstep_EaseOut}.Result
    B=VectorParameter() {parameterName=FlashColor, defaultValue=<rgba 0.1258 0.0113 0.14 0>}.RGB
   )
   B=Add(A=Multiply(A=@n4.RGB, B=@n6), B=Multiply(A=@n28, B=@n29))
  )
  Constant() {r=1}
  n30 = CollectionParameter() {collection=&FortniteMaterialParameters, parameterName=SpookyMistMode}
  n31 = ScalarParameter() {parameterName=MovementOpacity}
  n32 = Add(A=Clamp(None=@n30, Max=@n31), B=@n31)
  Multiply(A=Multiply(A=@n26, B=Clamp(None=@n32)), B=ScalarParameter() {defaultValue=0.84, parameterName=OverallOpacity})
  OneMinus(None=@n21.G)
  n33 = Multiply(A=Constant3Vector() {constant=<rgba -0.08 0.08 0.6 0>}, B=@n16)
  QualitySwitch(
   Default=MaterialFunctionCall(
    TextureObject=TextureObject() {samplerType=SAMPLERTYPE_Normal, texture=&T_Weap_Distort_N_02}
    TextureSize=@n15
    WorldPosition=Add(A=@n33, B=@n18)
   ) {materialFunction=&WorldAlignedNormal}.XYZ Texture
   Low=Constant()
  )
 ]
}
