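# Niagara effect system for the SpookyMist foraged item, named as its pickup effect.
# Five emitters, named for a flare and light, a trail, two swirl layers and small GPU bits, make up the effect.
# Four user color parameters default to red-pink tints, one strongly overbright.
# Only the SystemState module is listed; emitter renderers and color bindings are not shown.
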NiagaraSystem /Game/Athena/Items/ForagedItems/SpookyMist/FX/NS_SpookyMist_Pickup
dependencies {usedDataInterfaces [], usedDynamicInputs [], usedModules [&SystemState]}
emitters {
 Flare_and_Light {rendererClasses [], renderers []}
 Small_GPU_bits {rendererClasses [], renderers []}
 Swirls {rendererClasses [], renderers []}
 Swirls001 {rendererClasses [], renderers []}
 Trail {rendererClasses [], renderers []}
}
summary {
 systemName=NS_SpookyMist_Pickup
 emitters [
  {emitterName=Flare_and_Light, rendererClasses []}
  {emitterName=Trail, rendererClasses []}
  {emitterName=Swirls, rendererClasses []}
  {emitterName=Small_GPU_bits, rendererClasses []}
  {emitterName=Swirls001, rendererClasses []}
 ]
 userVariables [
  User.Color_a {type=LinearColor, description="", defaultValue {struct=LinearColor, value=<rgba 1.5 0.0652 0.1848 1.3>}}
  User.Color_b {type=LinearColor, description="", defaultValue {struct=LinearColor, value=<rgba 1.5 0 0.024 1>}}
  User.Color_c {type=LinearColor, description="", defaultValue {struct=LinearColor, value=<rgba 1 0 0.0427 1.5>}}
  User.Color_d {type=LinearColor, description="", defaultValue {struct=LinearColor, value=<rgba 35 1 1.4939 2>}}
 ]
}
