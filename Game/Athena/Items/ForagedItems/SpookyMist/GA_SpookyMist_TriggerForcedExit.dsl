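; Gameplay ability for the SpookyMist foraged item that lets a player force an early ghost mode exit by holding input.
; On press, unless driving, it starts the exit and a timer, both using the scalable CancelDelay.
; On expiry, it ends the exit phase and sends the Abilities.Generic.TriggeredAbility.Activate gameplay event to the pawn.
; Early release ends the exit phase and clears the timer; a helper caches the avatar pawn.
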(event Ability|EventOnAbilityInputPressed
  (bind _canceldelay (Variables|Default|GetCancelDelay))
  (bind _returnvalue (ScalableFloat|GetValueatLevel _canceldelay))
  (bind _outfortplayerpawn (CallFunction|GetValidFortPawn))
  (if (not (Vehicle|IsDrivingVehicle _outfortplayerpawn))
    (GhostMode|StartGhostModeExit (Variables|Default|GetPlayerPawn) _returnvalue)
    (bind _returnvalue_2 (Utilities|Time|SetTimerbyEvent (delegate Custom|CanCancel) _returnvalue))
    (Variables|Default|SetCanCancelTimer _returnvalue_2)))

(event Custom|CanCancel
  (bind _playerpawn (Variables|Default|GetPlayerPawn))
  (GhostMode|EndGhostModeExit _playerpawn)
  (Ability|SendGameplayEventtoActor _playerpawn "(TagName=\"Abilities.Generic.TriggeredAbility.Activate\")" (Utilities|Struct|MakeGameplayEventData "(TagName=\"Abilities.Generic.TriggeredAbility.Activate\")")))

(event Ability|EventOnAbilityInputReleased
  (bind _outfortplayerpawn (CallFunction|GetValidFortPawn))
  (GhostMode|EndGhostModeExit _outfortplayerpawn)
  (Utilities|Time|ClearandInvalidateTimerbyHandle (Variables|Default|GetCanCancelTimer)))

(fn GetValidFortPawn
  (Utilities|IsValid (Variables|Default|GetPlayerPawn)
    (:Is%20Valid
      _)
    (:Is%20Not%20Valid
      (bind _outfortplayerpawn (CallFunction|GetAvatarasValidFortPlayerPawn)
        (:ValidFortPlayerPawn
          (Variables|Default|SetPlayerPawn _outfortplayerpawn)))))
  (return (Variables|Default|GetPlayerPawn)))
